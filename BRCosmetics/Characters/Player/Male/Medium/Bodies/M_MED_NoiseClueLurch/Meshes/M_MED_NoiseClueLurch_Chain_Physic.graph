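# Physics asset for the right-side dangling chain on the male medium NoiseClueLurch player body.
# Capsule bodies cover four chain bones, dyn_chain_r_02 to 05, alongside pelvis, right thigh and right calf capsules that give the leg and hip volume.
# The chain's first link attaches to the thigh with free swing and twist, and later links join in sequence with tighter angular limits.
# The calf-to-thigh joint is also free, while the pelvis has no constraint.
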PhysicsAsset /BRCosmetics/Characters/Player/Male/Medium/Bodies/M_MED_NoiseClueLurch/Meshes/M_MED_NoiseClueLurch_Chain_Physic
bodies {
 calf_r [{shapeType=Capsule, length=32, radius=9, shapeName=thigh_r_capsule, center=<23.1561 3.807 -1.9153>, rotation=<rot 90 0 0>}]
 dyn_chain_r_02 [{shapeType=Capsule, length=6, radius=3, center=<4 0 0>}]
 dyn_chain_r_03 [{shapeType=Capsule, length=12, radius=3, center=<4.7653 10 -3>, rotation=<rot 90 159.9999 0>}]
 dyn_chain_r_04 [
  {shapeType=Capsule, length=6, radius=3, center=<5 6 -2>, rotation=<rot 80 139.9997 -0.0001>}
  {
   shapeType=Capsule
   length=4.5578
   radius=2.3396
   shapeName=dyn_chain_r_04_capsule
   center=<4.8204 -6.471 1.0813>
   rotation=<rot 80 139.9995 110.5773>
  }
 ]
 dyn_chain_r_05 [{shapeType=Capsule, length=6, radius=3, center=<4 0 0>, rotation=<rot 90 90 0>}]
 pelvis [{shapeType=Capsule, length=12.2058, radius=9.5, shapeName=pelvis_capsule, center=<0 2.6558 0>}]
 thigh_r [
  {shapeType=Capsule, length=32, radius=7, shapeName=thigh_r_capsule, center=<23.1561 -2.3525 -1.182>, rotation=<rot 90 0 0>}
  {shapeType=Capsule, length=32, radius=7, shapeName=thigh_r_capsule, center=<23.1561 3.807 -1.182>, rotation=<rot 90 0 0>}
 ]
}
constraints [
 {
  bone1Name=calf_r
  bone2Name=thigh_r
  swing1LimitDegrees=45
  swing1Motion=Free
  swing2LimitDegrees=45
  swing2Motion=Free
  twistLimitDegrees=45
  twistMotion=Free
 }
 {
  bone1Name=dyn_chain_r_02
  bone2Name=thigh_r
  swing1LimitDegrees=45
  swing1Motion=Free
  swing2LimitDegrees=45
  swing2Motion=Free
  twistLimitDegrees=45
  twistMotion=Free
 }
 {bone1Name=dyn_chain_r_03, bone2Name=dyn_chain_r_02, swing1LimitDegrees=20, swing2LimitDegrees=25, twistLimitDegrees=10}
 {bone1Name=dyn_chain_r_04, bone2Name=dyn_chain_r_03, swing1LimitDegrees=25, swing2LimitDegrees=25, twistLimitDegrees=10}
 {bone1Name=dyn_chain_r_05, bone2Name=dyn_chain_r_04, swing1LimitDegrees=25, swing2LimitDegrees=25, twistLimitDegrees=10}
]
