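# Physics asset for the NoiseClueLurch medium male player cosmetic body, set up for its dangling tie and front skirt bones.
# Capsules on the pelvis, spine, clavicles and thighs form the body volume around the dynamic bones.
# A six-link tie chain hangs from a tiny base body on spine_04, with moderate swing and very tight twist limits per link.
# The tie base and both lower skirt bodies rotate freely, the skirt pieces constrained directly to the pelvis.
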PhysicsAsset /BRCosmetics/Characters/Player/Male/Medium/Bodies/M_MED_NoiseClueLurch/Meshes/M_MED_NoiseClueLurch_Physics
bodies {
 base_dyn_tie [{shapeType=Capsule, length=1, radius=1}]
 clavicle_l [{shapeType=Capsule, length=8.7147, radius=8.7147, shapeName=clavicle_l_capsule, center=<7.5297 0 -0.8825>, rotation=<rot -90 0 0>}]
 clavicle_r [{shapeType=Capsule, length=8.7147, radius=8.7147, shapeName=clavicle_r_capsule, center=<-7.5297 0 0.8825>, rotation=<rot -90 0 0>}]
 dyn_front_skirt_l_bott [{shapeType=Capsule, length=9.0798, radius=5.9202, shapeName=dyn_front_skirt_l_bott_capsule, rotation=<rot 90 0 0>}]
 dyn_front_skirt_r_bott [{shapeType=Capsule, length=9.0798, radius=5.9202, shapeName=dyn_front_skirt_r_bott_capsule, rotation=<rot 90 0 0>}]
 dyn_tie_01 [{shapeType=Capsule, length=6, radius=4, center=<4 0 3.5>, rotation=<rot 90 0 90>}]
 dyn_tie_02 [{shapeType=Capsule, length=6, radius=4, center=<4 0 3>, rotation=<rot 90 0 90>}]
 dyn_tie_03 [{shapeType=Capsule, length=6, radius=4, center=<4 0 2.9016>, rotation=<rot 90 -90 0>}]
 dyn_tie_04 [{shapeType=Capsule, length=6, radius=4, center=<4 0 3.2769>, rotation=<rot 90 -90 0>}]
 dyn_tie_05 [{shapeType=Capsule, length=6, radius=4, center=<4 0 3.5659>, rotation=<rot 90 -90 0>}]
 dyn_tie_06 [{shapeType=Capsule, length=12, radius=5, center=<4 0 4.2707>, rotation=<rot 90 -90 0>}]
 pelvis [
  {shapeType=Capsule, length=13.7914, radius=14.164, shapeName=pelvis_capsule, center=<-2.235 0.2935 0>}
  {shapeType=Capsule, length=12.8828, radius=15.0726, shapeName=pelvis_capsule, center=<-11.0761 1.0838 0>}
 ]
 spine_01 [{shapeType=Capsule, length=11.5795, radius=11.8001, shapeName=spine_01_capsule, center=<1.6624 0.6182 0>}]
 spine_02 [{shapeType=Capsule, length=11.6898, radius=11.6898, shapeName=spine_02_capsule, center=<3.1006 0 0>}]
 spine_03 [{shapeType=Capsule, length=12.1022, radius=12.1022, shapeName=spine_03_capsule, center=<5.1166 0 0>}]
 spine_04 [{shapeType=Capsule, length=12.1022, radius=12.1022, shapeName=spine_04_capsule, center=<5.1166 0 0>}]
 thigh_l [{shapeType=Capsule, length=29.6066, radius=14.3934, shapeName=thigh_l_capsule, center=<-25.0807 0.145 0>, rotation=<rot -90 3.6469 0>}]
 thigh_r [{shapeType=Capsule, length=30.2387, radius=14.7613, shapeName=thigh_r_capsule, center=<25.8249 0 0>, rotation=<rot -90 3.8475 0>}]
}
constraints [
 {
  bone1Name=base_dyn_tie
  bone2Name=spine_04
  swing1LimitDegrees=45
  swing1Motion=Free
  swing2LimitDegrees=45
  swing2Motion=Free
  twistLimitDegrees=45
  twistMotion=Free
 }
 {bone1Name=dyn_tie_01, bone2Name=base_dyn_tie, swing1LimitDegrees=35, swing2LimitDegrees=25, twistLimitDegrees=2}
 {bone1Name=dyn_tie_02, bone2Name=dyn_tie_01, swing1LimitDegrees=35, swing2LimitDegrees=25, twistLimitDegrees=2}
 {bone1Name=dyn_tie_03, bone2Name=dyn_tie_02, swing1LimitDegrees=35, swing2LimitDegrees=25, twistLimitDegrees=2}
 {bone1Name=dyn_tie_04, bone2Name=dyn_tie_03, swing1LimitDegrees=35, swing2LimitDegrees=25, twistLimitDegrees=2}
 {bone1Name=dyn_tie_05, bone2Name=dyn_tie_04, swing1LimitDegrees=35, swing2LimitDegrees=25, twistLimitDegrees=2}
 {bone1Name=dyn_tie_06, bone2Name=dyn_tie_05, swing1LimitDegrees=35, swing2LimitDegrees=25, twistLimitDegrees=2}
 {
  bone1Name=dyn_front_skirt_l_bott
  bone2Name=pelvis
  swing1LimitDegrees=45
  swing1Motion=Free
  swing2LimitDegrees=45
  swing2Motion=Free
  twistLimitDegrees=45
  twistMotion=Free
 }
 {
  bone1Name=dyn_front_skirt_r_bott
  bone2Name=pelvis
  swing1LimitDegrees=45
  swing1Motion=Free
  swing2LimitDegrees=45
  swing2Motion=Free
  twistLimitDegrees=45
  twistMotion=Free
 }
]
