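# Physics asset for the face accessory mesh of the male medium NoiseClueLurch player cosmetic, set up for secondary motion of its hair.
# Capsules on spine, arms and neck, plus capsules and a box on the head, form body collision shapes.
# Capsule chains for center, left, right, front and side hair hang from head-linked base bones with tight angular limits, some with locked twist.
# Free-rotation constraints cross-link the center chain to left and right.
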PhysicsAsset /BRCosmetics/Characters/Player/Male/Medium/Bodies/M_MED_NoiseClueLurch/Meshes/Parts/M_MED_NoiseClueLurch_FaceAcc_Physics
bodies {
 base_hair_c [{shapeType=Capsule, length=0.505, radius=0.505, shapeName=base_hair_c_capsule, rotation=<rot 0 -90 0>}]
 base_hair_front_l [{shapeType=Capsule, length=0.505, radius=0.505, shapeName=base_hair_front_l_capsule, rotation=<rot 0 -90 0>}]
 base_hair_front_r [{shapeType=Capsule, length=0.505, radius=0.505, shapeName=base_hair_front_r_capsule, rotation=<rot 0 -90 0>}]
 base_hair_front_r_b [{shapeType=Capsule, length=0.505, radius=0.505, shapeName=base_hair_front_r_b_capsule, rotation=<rot 0 -90 0>}]
 base_hair_l [{shapeType=Capsule, length=0.505, radius=0.505, shapeName=base_hair_l_capsule, rotation=<rot 0 -90 0>}]
 base_hair_r [{shapeType=Capsule, length=0.505, radius=0.505, shapeName=base_hair_r_capsule, rotation=<rot 0 -90 0>}]
 base_hair_side_l [{shapeType=Capsule, length=1, radius=1}]
 base_hair_side_r [{shapeType=Capsule, length=1, radius=1}]
 base_hair_side_r_s [{shapeType=Capsule, length=1, radius=1}]
 clavicle_l [{shapeType=Capsule, length=8.7905, radius=8, shapeName=clavicle_l_capsule, center=<8.094 -0.6251 -0.4529>, rotation=<rot -95 0 0>}]
 clavicle_r [{shapeType=Capsule, length=8.7905, radius=9, shapeName=clavicle_r_capsule, center=<-8.094 0.6251 0.4529>, rotation=<rot -95 0 0>}]
 dyn_hair_c_01 [{shapeType=Capsule, length=6, radius=3.5, center=<4 0 3>, rotation=<rot 0 0 90>}]
 dyn_hair_c_02 [{shapeType=Capsule, length=6, radius=3.5, center=<4 0 3>, rotation=<rot 0 0 90>}]
 dyn_hair_c_03 [{shapeType=Capsule, length=6, radius=3.5, center=<4 0 3>, rotation=<rot 0 0 90>}]
 dyn_hair_c_04 [{shapeType=Capsule, length=6, radius=4, center=<4 0 1>, rotation=<rot 0 0 90>}]
 dyn_hair_c_05 [{shapeType=Capsule, length=6, radius=4.5, center=<4 0 1>, rotation=<rot 0 0 90>}]
 dyn_hair_front_l_01 [{shapeType=Capsule, length=5, radius=2.5, center=<3 0 3>, rotation=<rot -90 0 90>}]
 dyn_hair_front_l_02 [{shapeType=Capsule, length=5, radius=2.5, center=<3 0 3>, rotation=<rot -90 0 0>}]
 dyn_hair_front_l_03 [{shapeType=Capsule, length=5, radius=2.5, center=<3 0 3>, rotation=<rot -90 0 0>}]
 dyn_hair_front_l_04 [{shapeType=Capsule, length=5, radius=2.5, center=<3 0 3>, rotation=<rot -90 0 0>}]
 dyn_hair_front_l_05 [{shapeType=Capsule, length=5, radius=4, center=<3 0 0>, rotation=<rot -90 0 0>}]
 dyn_hair_front_l_06 [{shapeType=Capsule, length=5, radius=3, center=<3 0 0>, rotation=<rot -90 0 0>}]
 dyn_hair_front_r_01 [{shapeType=Capsule, length=4, radius=2.5, center=<3 0 2>, rotation=<rot 0 0 90>}]
 dyn_hair_front_r_01_b [{shapeType=Capsule, length=3, radius=2, center=<3 0 0>, rotation=<rot -90 -90 0>}]
 dyn_hair_front_r_02 [{shapeType=Capsule, length=4, radius=2.5, center=<3 0 2>, rotation=<rot 0 0 130>}]
 dyn_hair_front_r_02_b [{shapeType=Capsule, length=3, radius=2, center=<3 0 1.5>, rotation=<rot -90 0 0>}]
 dyn_hair_front_r_03 [{shapeType=Capsule, length=4, radius=2.5, center=<3 0 2>, rotation=<rot 0 0 130>}]
 dyn_hair_front_r_03_b [{shapeType=Capsule, length=3, radius=2, center=<3 0 1.5>, rotation=<rot -90 0 0>}]
 dyn_hair_front_r_04 [{shapeType=Capsule, length=10, radius=3.5, center=<3 0 0>, rotation=<rot 0.0001 -0.0001 105>}]
 dyn_hair_front_r_04_b [{shapeType=Capsule, length=3, radius=2, center=<3 0 1>, rotation=<rot 90 0 0>}]
 dyn_hair_front_r_05 [{shapeType=Capsule, length=4, radius=2, center=<3 0 0>, rotation=<rot 0 0 90>}]
 dyn_hair_front_r_05_b [{shapeType=Capsule, length=3, radius=1.5, center=<3 0 0>, rotation=<rot -90 0 0>}]
 dyn_hair_front_r_05_s [{shapeType=Capsule, length=4, radius=2.5, center=<3 0 2>, rotation=<rot 0 0 90>}]
 dyn_hair_front_r_06 [{shapeType=Capsule, length=4, radius=2.5, center=<3 0 0>, rotation=<rot 0 0 90>}]
 dyn_hair_front_r_06_b [{shapeType=Capsule, length=3, radius=1.5, center=<3 0 0>, rotation=<rot -90 0 0>}]
 dyn_hair_front_r_07 [{shapeType=Capsule, length=5, radius=3, center=<3 0 2>, rotation=<rot 90 0 0>}]
 dyn_hair_l_01 [{shapeType=Capsule, length=6, radius=2.5, center=<4 0 3>, rotation=<rot 0 0 90>}]
 dyn_hair_l_02 [{shapeType=Capsule, length=6, radius=3.5, center=<4 0 0>, rotation=<rot 0 0 90>}]
 dyn_hair_l_03 [{shapeType=Capsule, length=6, radius=3.5, center=<4 0 0>, rotation=<rot 0 0 90>}]
 dyn_hair_l_04 [{shapeType=Capsule, length=6, radius=4, center=<4 0 1>, rotation=<rot 0 0 90>}]
 dyn_hair_l_05 [{shapeType=Capsule, length=6, radius=4, center=<4 0 2>, rotation=<rot 0 0 90>}]
 dyn_hair_r_01 [{shapeType=Capsule, length=6, radius=3, center=<4 0 3>, rotation=<rot 0 0 90>}]
 dyn_hair_r_02 [{shapeType=Capsule, length=6, radius=3.5, center=<4 0 0>, rotation=<rot 0 0 90>}]
 dyn_hair_r_03 [{shapeType=Capsule, length=6, radius=3.5, center=<4 0 3>, rotation=<rot 0 0 90>}]
 dyn_hair_r_04 [{shapeType=Capsule, length=6, radius=4, center=<4 0 2>, rotation=<rot 0 0 90>}]
 dyn_hair_r_05 [{shapeType=Capsule, length=6, radius=4, center=<4 0 2>, rotation=<rot 0 0 90>}]
 dyn_hair_r_06_s [{shapeType=Capsule, length=3, radius=3.5, center=<4 0 2>}]
 dyn_hair_r_07_s [{shapeType=Capsule, length=6, radius=3.5, center=<4 0 2>, rotation=<rot 0 0 90>}]
 dyn_hair_r_08_s [{shapeType=Capsule, length=6, radius=3.5, center=<4 0 2>, rotation=<rot 0 0 90>}]
 dyn_hair_side_l_01 [{shapeType=Capsule, length=6, radius=3, center=<2 0 2>, rotation=<rot 90 0 90>}]
 dyn_hair_side_l_02 [{shapeType=Capsule, length=6, radius=3, center=<2 0 2>, rotation=<rot 90 0 90>}]
 dyn_hair_side_l_03 [{shapeType=Capsule, length=6, radius=3, center=<2 0 2>, rotation=<rot 90 0 90>}]
 dyn_hair_side_l_04 [{shapeType=Capsule, length=6, radius=3, center=<2 0 2>, rotation=<rot 90 0 90>}]
 dyn_hair_side_l_05 [{shapeType=Capsule, length=6, radius=2.5, center=<2 0 2>, rotation=<rot 90 0 90>}]
 dyn_hair_side_r_01 [{shapeType=Capsule, length=6, radius=2, center=<2 0 1>, rotation=<rot 90 0 90>}]
 dyn_hair_side_r_01_s [{shapeType=Capsule, length=6, radius=2, center=<2 0 4>, rotation=<rot 90 0 90>}]
 dyn_hair_side_r_02 [{shapeType=Capsule, length=6, radius=3, center=<2 0 1>, rotation=<rot 90 0 90>}]
 dyn_hair_side_r_02_s [{shapeType=Capsule, length=4, radius=2, center=<2 0 4>, rotation=<rot 90 0 90>}]
 dyn_hair_side_r_03 [{shapeType=Capsule, length=6, radius=3, center=<2 0 1>, rotation=<rot 90 0 90>}]
 dyn_hair_side_r_03_s [{shapeType=Capsule, length=4, radius=2.5, center=<2 0 3>, rotation=<rot 0 90 0>}]
 head [
  {shapeType=Capsule, length=7.4648, radius=6, shapeName=head_capsule, center=<3.2863 -1.3415 0>, rotation=<rot -90 0 0>}
  {shapeType=Capsule, length=7.4648, radius=4, shapeName=head_capsule, center=<2.4391 6.7132 0>, rotation=<rot -90 -15 0>}
  {shapeType=Box, extentX=6.5, extentY=13.5, extentZ=14.5, shapeName=head_box, center=<3.3775 4.4041 0>}
 ]
 lowerarm_l [{shapeType=Capsule, length=22, radius=6.5264, shapeName=upperarm_l_capsule, center=<10 0 0>, rotation=<rot -90 0 0>}]
 lowerarm_r [{shapeType=Capsule, length=22, radius=6.5264, shapeName=upperarm_r_capsule, center=<-10 0 0>, rotation=<rot -90 0 0>}]
 neck_01 [{shapeType=Capsule, length=6.9272, radius=5, shapeName=neck_01_capsule, rotation=<rot -90 0 0>}]
 neck_02 [{shapeType=Capsule, length=6.7048, radius=5, shapeName=neck_02_capsule, rotation=<rot -90 0 0>}]
 spine_03 [{shapeType=Capsule, length=11.0817, radius=9, shapeName=spine_05_capsule, center=<1.0603 -0.7063 0>, rotation=<rot 0 -0.0001 0>}]
 spine_04 [{shapeType=Capsule, length=12, radius=12, shapeName=spine_05_capsule, center=<1.0603 -0.7063 0>}]
 spine_05 [{shapeType=Capsule, length=11.0817, radius=9, shapeName=spine_05_capsule, center=<1.0603 -0.7063 0>}]
 upperarm_l [{shapeType=Capsule, length=22, radius=6.5264, shapeName=upperarm_l_capsule, center=<10 0 0>, rotation=<rot -90 0 0>}]
 upperarm_r [{shapeType=Capsule, length=22, radius=6.5264, shapeName=upperarm_r_capsule, center=<-10 0 0>, rotation=<rot -90 0 0>}]
}
constraints [
 {
  bone1Name=spine_04
  bone2Name=spine_03
  swing1LimitDegrees=45
  swing1Motion=Free
  swing2LimitDegrees=45
  swing2Motion=Free
  twistLimitDegrees=45
  twistMotion=Free
 }
 {
  bone1Name=spine_05
  bone2Name=spine_04
  swing1LimitDegrees=45
  swing1Motion=Free
  swing2LimitDegrees=45
  swing2Motion=Free
  twistLimitDegrees=45
  twistMotion=Free
 }
 {
  bone1Name=clavicle_l
  bone2Name=spine_04
  swing1LimitDegrees=45
  swing1Motion=Free
  swing2LimitDegrees=45
  swing2Motion=Free
  twistLimitDegrees=45
  twistMotion=Free
 }
 {
  bone1Name=upperarm_l
  bone2Name=spine_04
  swing1LimitDegrees=45
  swing1Motion=Free
  swing2LimitDegrees=45
  swing2Motion=Free
  twistLimitDegrees=45
  twistMotion=Free
 }
 {
  bone1Name=lowerarm_l
  bone2Name=upperarm_l
  swing1LimitDegrees=45
  swing1Motion=Free
  swing2LimitDegrees=45
  swing2Motion=Free
  twistLimitDegrees=45
  twistMotion=Free
 }
 {
  bone1Name=clavicle_r
  bone2Name=spine_04
  swing1LimitDegrees=45
  swing1Motion=Free
  swing2LimitDegrees=45
  swing2Motion=Free
  twistLimitDegrees=45
  twistMotion=Free
 }
 {
  bone1Name=upperarm_r
  bone2Name=spine_04
  swing1LimitDegrees=45
  swing1Motion=Free
  swing2LimitDegrees=45
  swing2Motion=Free
  twistLimitDegrees=45
  twistMotion=Free
 }
 {
  bone1Name=lowerarm_r
  bone2Name=upperarm_r
  swing1LimitDegrees=45
  swing1Motion=Free
  swing2LimitDegrees=45
  swing2Motion=Free
  twistLimitDegrees=45
  twistMotion=Free
 }
 {
  bone1Name=neck_01
  bone2Name=spine_04
  swing1LimitDegrees=45
  swing1Motion=Free
  swing2LimitDegrees=45
  swing2Motion=Free
  twistLimitDegrees=45
  twistMotion=Free
 }
 {
  bone1Name=neck_02
  bone2Name=spine_04
  swing1LimitDegrees=45
  swing1Motion=Free
  swing2LimitDegrees=45
  swing2Motion=Free
  twistLimitDegrees=45
  twistMotion=Free
 }
 {
  bone1Name=head
  bone2Name=spine_04
  swing1LimitDegrees=45
  swing1Motion=Free
  swing2LimitDegrees=45
  swing2Motion=Free
  twistLimitDegrees=45
  twistMotion=Free
 }
 {
  bone1Name=base_hair_c
  bone2Name=head
  swing1LimitDegrees=45
  swing1Motion=Free
  swing2LimitDegrees=45
  swing2Motion=Free
  twistLimitDegrees=45
  twistMotion=Free
 }
 {bone1Name=dyn_hair_c_01, bone2Name=base_hair_c, swing1LimitDegrees=17.5, swing2LimitDegrees=20, twistLimitDegrees=2}
 {bone1Name=dyn_hair_c_02, bone2Name=dyn_hair_c_01, swing1LimitDegrees=17.5, swing2LimitDegrees=20, twistLimitDegrees=2}
 {bone1Name=dyn_hair_c_03, bone2Name=dyn_hair_c_02, swing1LimitDegrees=17.5, swing2LimitDegrees=20, twistLimitDegrees=2}
 {bone1Name=dyn_hair_c_04, bone2Name=dyn_hair_c_03, swing1LimitDegrees=17.5, swing2LimitDegrees=20, twistLimitDegrees=2}
 {
  bone1Name=dyn_hair_l_04
  bone2Name=dyn_hair_c_04
  swing1LimitDegrees=45
  swing1Motion=Free
  swing2LimitDegrees=45
  swing2Motion=Free
  twistLimitDegrees=45
  twistMotion=Free
 }
 {
  bone1Name=dyn_hair_r_04
  bone2Name=dyn_hair_c_04
  swing1LimitDegrees=45
  swing1Motion=Free
  swing2LimitDegrees=45
  swing2Motion=Free
  twistLimitDegrees=45
  twistMotion=Free
 }
 {bone1Name=dyn_hair_c_05, bone2Name=dyn_hair_c_04, swing1LimitDegrees=17.5, swing2LimitDegrees=20, twistLimitDegrees=2}
 {
  bone1Name=base_hair_front_l
  bone2Name=head
  swing1LimitDegrees=45
  swing1Motion=Free
  swing2LimitDegrees=45
  swing2Motion=Free
  twistLimitDegrees=45
  twistMotion=Free
 }
 {bone1Name=dyn_hair_front_l_01, bone2Name=base_hair_front_l, swing1LimitDegrees=10, swing2LimitDegrees=25, twistLimitDegrees=2}
 {bone1Name=dyn_hair_front_l_02, bone2Name=dyn_hair_front_l_01, swing1LimitDegrees=25, swing2LimitDegrees=25, twistLimitDegrees=2}
 {bone1Name=dyn_hair_front_l_03, bone2Name=dyn_hair_front_l_02, swing1LimitDegrees=25, swing2LimitDegrees=25, twistLimitDegrees=2}
 {bone1Name=dyn_hair_front_l_04, bone2Name=dyn_hair_front_l_03, swing1LimitDegrees=25, swing2LimitDegrees=25, twistLimitDegrees=2}
 {bone1Name=dyn_hair_front_l_05, bone2Name=dyn_hair_front_l_04, swing1LimitDegrees=25, swing2LimitDegrees=25, twistLimitDegrees=2}
 {bone1Name=dyn_hair_front_l_06, bone2Name=dyn_hair_front_l_05, swing1LimitDegrees=25, swing2LimitDegrees=25, twistLimitDegrees=2}
 {
  bone1Name=base_hair_front_r
  bone2Name=head
  swing1LimitDegrees=45
  swing1Motion=Free
  swing2LimitDegrees=45
  swing2Motion=Free
  twistLimitDegrees=45
  twistMotion=Free
 }
 {
  bone1Name=dyn_hair_front_r_01
  bone2Name=base_hair_front_r
  swing1LimitDegrees=15
  swing2LimitDegrees=10
  twistLimitDegrees=10
  twistMotion=Locked
 }
 {
  bone1Name=dyn_hair_front_r_02
  bone2Name=dyn_hair_front_r_01
  swing1LimitDegrees=30
  swing2LimitDegrees=25
  twistLimitDegrees=10
  twistMotion=Locked
 }
 {
  bone1Name=dyn_hair_front_r_03
  bone2Name=dyn_hair_front_r_02
  swing1LimitDegrees=30
  swing2LimitDegrees=25
  twistLimitDegrees=10
  twistMotion=Locked
 }
 {
  bone1Name=dyn_hair_front_r_04
  bone2Name=dyn_hair_front_r_03
  swing1LimitDegrees=30
  swing2LimitDegrees=25
  twistLimitDegrees=10
  twistMotion=Locked
 }
 {
  bone1Name=dyn_hair_front_r_05
  bone2Name=dyn_hair_front_r_04
  swing1LimitDegrees=30
  swing2LimitDegrees=65
  twistLimitDegrees=10
  twistMotion=Locked
 }
 {
  bone1Name=dyn_hair_front_r_06
  bone2Name=dyn_hair_front_r_05
  swing1LimitDegrees=30
  swing2LimitDegrees=65
  twistLimitDegrees=10
  twistMotion=Locked
 }
 {
  bone1Name=dyn_hair_front_r_07
  bone2Name=dyn_hair_front_r_06
  swing1LimitDegrees=30
  swing2LimitDegrees=25
  twistLimitDegrees=10
  twistMotion=Locked
 }
 {
  bone1Name=dyn_hair_front_r_05_s
  bone2Name=dyn_hair_front_r_05
  swing1LimitDegrees=30
  swing2LimitDegrees=25
  twistLimitDegrees=10
  twistMotion=Locked
 }
 {
  bone1Name=base_hair_front_r_b
  bone2Name=head
  swing1LimitDegrees=45
  swing1Motion=Free
  swing2LimitDegrees=45
  swing2Motion=Free
  twistLimitDegrees=45
  twistMotion=Free
 }
 {bone1Name=dyn_hair_front_r_01_b, bone2Name=base_hair_front_r_b, swing1LimitDegrees=10, swing2LimitDegrees=10, twistLimitDegrees=2}
 {bone1Name=dyn_hair_front_r_02_b, bone2Name=dyn_hair_front_r_01_b, swing1LimitDegrees=10, swing2LimitDegrees=30, twistLimitDegrees=2}
 {bone1Name=dyn_hair_front_r_03_b, bone2Name=dyn_hair_front_r_02_b, swing1LimitDegrees=10, swing2LimitDegrees=35, twistLimitDegrees=2}
 {bone1Name=dyn_hair_front_r_04_b, bone2Name=dyn_hair_front_r_03_b, swing1LimitDegrees=10, swing2LimitDegrees=35, twistLimitDegrees=2}
 {bone1Name=dyn_hair_front_r_05_b, bone2Name=dyn_hair_front_r_04_b, swing1LimitDegrees=10, swing2LimitDegrees=35, twistLimitDegrees=2}
 {bone1Name=dyn_hair_front_r_06_b, bone2Name=dyn_hair_front_r_05_b, swing1LimitDegrees=10, swing2LimitDegrees=35, twistLimitDegrees=2}
 {
  bone1Name=base_hair_l
  bone2Name=head
  swing1LimitDegrees=45
  swing1Motion=Free
  swing2LimitDegrees=45
  swing2Motion=Free
  twistLimitDegrees=45
  twistMotion=Free
 }
 {bone1Name=dyn_hair_l_01, bone2Name=base_hair_l, swing1LimitDegrees=17.5, swing2LimitDegrees=20, twistLimitDegrees=2}
 {bone1Name=dyn_hair_l_02, bone2Name=dyn_hair_l_01, swing1LimitDegrees=17.5, swing2LimitDegrees=20, twistLimitDegrees=2}
 {bone1Name=dyn_hair_l_03, bone2Name=dyn_hair_l_02, swing1LimitDegrees=17.5, swing2LimitDegrees=20, twistLimitDegrees=2}
 {bone1Name=dyn_hair_l_04, bone2Name=dyn_hair_l_03, swing1LimitDegrees=17.5, swing2LimitDegrees=20, twistLimitDegrees=2}
 {bone1Name=dyn_hair_l_05, bone2Name=dyn_hair_l_04, swing1LimitDegrees=17.5, swing2LimitDegrees=20, twistLimitDegrees=2}
 {
  bone1Name=base_hair_r
  bone2Name=head
  swing1LimitDegrees=45
  swing1Motion=Free
  swing2LimitDegrees=45
  swing2Motion=Free
  twistLimitDegrees=45
  twistMotion=Free
 }
 {bone1Name=dyn_hair_r_01, bone2Name=base_hair_r, swing1LimitDegrees=17.5, swing2LimitDegrees=20, twistLimitDegrees=2}
 {bone1Name=dyn_hair_r_02, bone2Name=dyn_hair_r_01, swing1LimitDegrees=17.5, swing2LimitDegrees=20, twistLimitDegrees=2}
 {bone1Name=dyn_hair_r_03, bone2Name=dyn_hair_r_02, swing1LimitDegrees=17.5, swing2LimitDegrees=20, twistLimitDegrees=2}
 {bone1Name=dyn_hair_r_04, bone2Name=dyn_hair_r_03, swing1LimitDegrees=17.5, swing2LimitDegrees=20, twistLimitDegrees=2}
 {bone1Name=dyn_hair_r_05, bone2Name=dyn_hair_r_04, swing1LimitDegrees=17.5, swing2LimitDegrees=20, twistLimitDegrees=2}
 {bone1Name=dyn_hair_r_06_s, bone2Name=dyn_hair_r_02, swing1LimitDegrees=15, swing2LimitDegrees=35, twistLimitDegrees=2}
 {bone1Name=dyn_hair_r_07_s, bone2Name=dyn_hair_r_06_s, swing1LimitDegrees=15, swing2LimitDegrees=25, twistLimitDegrees=2}
 {bone1Name=dyn_hair_r_08_s, bone2Name=dyn_hair_r_07_s, swing1LimitDegrees=15, swing2LimitDegrees=25, twistLimitDegrees=2}
 {
  bone1Name=base_hair_side_l
  bone2Name=head
  swing1LimitDegrees=45
  swing1Motion=Free
  swing2LimitDegrees=45
  swing2Motion=Free
  twistLimitDegrees=45
  twistMotion=Free
 }
 {bone1Name=dyn_hair_side_l_01, bone2Name=base_hair_side_l, swing1LimitDegrees=20, swing2LimitDegrees=15, twistLimitDegrees=2}
 {bone1Name=dyn_hair_side_l_02, bone2Name=dyn_hair_side_l_01, swing1LimitDegrees=20, swing2LimitDegrees=10, twistLimitDegrees=2}
 {bone1Name=dyn_hair_side_l_03, bone2Name=dyn_hair_side_l_02, swing1LimitDegrees=20, swing2LimitDegrees=10, twistLimitDegrees=2}
 {bone1Name=dyn_hair_side_l_04, bone2Name=dyn_hair_side_l_03, swing1LimitDegrees=20, swing2LimitDegrees=10, twistLimitDegrees=2}
 {bone1Name=dyn_hair_side_l_05, bone2Name=dyn_hair_side_l_04, swing1LimitDegrees=20, swing2LimitDegrees=10, twistLimitDegrees=2}
 {
  bone1Name=base_hair_side_r
  bone2Name=head
  swing1LimitDegrees=45
  swing1Motion=Free
  swing2LimitDegrees=45
  swing2Motion=Free
  twistLimitDegrees=45
  twistMotion=Free
 }
 {bone1Name=dyn_hair_side_r_01, bone2Name=base_hair_side_r, swing1LimitDegrees=10, swing2LimitDegrees=15, twistLimitDegrees=2}
 {bone1Name=dyn_hair_side_r_02, bone2Name=dyn_hair_side_r_01, swing1LimitDegrees=25, swing2LimitDegrees=15, twistLimitDegrees=2}
 {bone1Name=dyn_hair_side_r_03, bone2Name=dyn_hair_side_r_02, swing1LimitDegrees=25, swing2LimitDegrees=15, twistLimitDegrees=2}
 {
  bone1Name=base_hair_side_r_s
  bone2Name=head
  swing1LimitDegrees=45
  swing1Motion=Free
  swing2LimitDegrees=45
  swing2Motion=Free
  twistLimitDegrees=45
  twistMotion=Free
 }
 {bone1Name=dyn_hair_side_r_01_s, bone2Name=base_hair_side_r_s, swing1LimitDegrees=10, swing2LimitDegrees=25, twistLimitDegrees=2}
 {bone1Name=dyn_hair_side_r_02_s, bone2Name=dyn_hair_side_r_01_s, swing1LimitDegrees=35, swing2LimitDegrees=25, twistLimitDegrees=2}
 {bone1Name=dyn_hair_side_r_03_s, bone2Name=dyn_hair_side_r_02_s, swing1LimitDegrees=35, swing2LimitDegrees=25, twistLimitDegrees=2}
]
