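# Post-process-style AnimBlueprint for the NoiseClueLurch character's FaceAcc part, animating its dynamic hair bones.
# It layers head-mesh facial bones over the copied parent pose, then runs Control Rig and neck-relative RigidBody physics into a cached pose.
# When backbling or capes are worn, or physics is disabled, output switches to hair bones constrained to head, neck, spine and clavicles, with small offsets.
# Events update gravity override, attachment flags and an unread sunglasses flag.
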AnimBlueprint /BRCosmetics/Characters/Player/Male/Medium/Bodies/M_MED_NoiseClueLurch/Meshes/Parts/M_MED_NoiseClueLurch_FaceAcc_AnimBP
target_skeleton=&/M_MED_NoiseClueLurch_FaceAcc_Skeleton
graphs [
 AnimGraph {nodes [
   n1 = Get GravityOverride()
   n2 = RigidBody!(ComponentPose=Local To Component(), Alpha, OverrideWorldGravity:Vector=@n1, ExternalForce) -> (Pose) {
    bClampLinearTranslationLimitToRefPose=true
    bOverrideWorldGravity=true
    overridePhysicsAsset=&/M_MED_NoiseClueLurch_FaceAcc_Physics
    simulationSpace=BaseBoneSpace
    alphaBoolBlend {blendOption=Linear}
    baseBoneRef {boneName=neck_01}
    componentAppliedLinearAccClamp=<0 0 0>
    overrideWorldGravity {z=-980}
    simSpaceSettings {
     dampingAlpha=0
     maxAngularAcceleration=24
     maxAngularVelocity=6
     maxLinearAcceleration=1200
     maxLinearVelocity=300
     velocityScaleZ=0.4
     worldAlpha=1
     externalLinearDragV=<0.15 0.15 0.15>
    }
   }
   Component To Local(ComponentPose=@n2)
   n3 = "Constraint - dyn_hair_r_01 (PP)"!(ComponentPose=Local To Component(LocalPose=Use cached pose 'Start_Cache'()), Alpha=1, ConstraintWeights_0=0.5, ConstraintWeights_1:float=0.5) -> (Pose) {
    alphaBoolBlend {blendOption=Linear}
    boneToModify {boneName=dyn_hair_r_01}
    constraintSetup [{targetBone {boneName=head}}, {targetBone {boneName=neck_02}}]
    constraintWeights [1,1]
   }
   n4 = "Constraint - dyn_hair_r_02 (PP)"!(ComponentPose=@n3, Alpha=1, ConstraintWeights_0=1, ConstraintWeights_1:float=0) -> (Pose) {
    alphaBoolBlend {blendOption=Linear}
    boneToModify {boneName=dyn_hair_r_02}
    constraintSetup [{targetBone {boneName=head}}, {targetBone {boneName=neck_02}}]
    constraintWeights [1,1]
   }
   n5 = "Constraint - dyn_hair_r_03 (PPP)"!(ComponentPose=@n4, Alpha=1, ConstraintWeights_0=0.5, ConstraintWeights_1:float=0.5, ConstraintWeights_2:float=0) -> (Pose) {
    alphaBoolBlend {blendOption=Linear}
    boneToModify {boneName=dyn_hair_r_03}
    constraintSetup [{targetBone {boneName=neck_01}}, {targetBone {boneName=clavicle_r}}, {targetBone {boneName=spine_05}}]
    constraintWeights [1,1,1]
   }
   n6 = "Constraint - dyn_hair_r_04 (PP)"!(ComponentPose=@n5, Alpha=1, ConstraintWeights_0=0.5, ConstraintWeights_1:float=0.5) -> (Pose) {
    alphaBoolBlend {blendOption=Linear}
    boneToModify {boneName=dyn_hair_r_04}
    constraintSetup [{targetBone {boneName=spine_05}}, {targetBone {boneName=clavicle_r}}]
    constraintWeights [1,1]
   }
   n7 = "Constraint - dyn_hair_r_05 (PP)"!(ComponentPose=@n6, Alpha=1, ConstraintWeights_0=1, ConstraintWeights_1:float=0) -> (Pose) {
    alphaBoolBlend {blendOption=Linear}
    boneToModify {boneName=dyn_hair_r_05}
    constraintSetup [{targetBone {boneName=spine_05}}, {targetBone {boneName=neck_01}}]
    constraintWeights [1,1]
   }
   n8 = "Constraint - dyn_hair_r_06_s (PPP)"!(ComponentPose=@n7, Alpha=1, ConstraintWeights_0=0, ConstraintWeights_1:float=0.5, ConstraintWeights_2:float=0.5) -> (Pose) {
    alphaBoolBlend {blendOption=Linear}
    boneToModify {boneName=dyn_hair_r_06_s}
    constraintSetup [{targetBone {boneName=neck_01}}, {targetBone {boneName=spine_05}}, {targetBone {boneName=clavicle_r}}]
    constraintWeights [1,1,1]
   }
   n9 = "Constraint - dyn_hair_r_07_s (PP)"!(ComponentPose=@n8, Alpha=1, ConstraintWeights_0=0.25, ConstraintWeights_1:float=0.75) -> (Pose) {
    alphaBoolBlend {blendOption=Linear}
    boneToModify {boneName=dyn_hair_r_07_s}
    constraintSetup [{targetBone {boneName=spine_05}}, {targetBone {boneName=clavicle_r}}]
    constraintWeights [1,1]
   }
   n10 = "Constraint - dyn_hair_r_08_s (PP)"!(ComponentPose=@n9, Alpha=1, ConstraintWeights_0=0.75, ConstraintWeights_1:float=0.25) -> (Pose) {
    alphaBoolBlend {blendOption=Linear}
    boneToModify {boneName=dyn_hair_r_08_s}
    constraintSetup [{targetBone {boneName=clavicle_r}}, {targetBone {boneName=spine_05}}]
    constraintWeights [1,1]
   }
   n11 = "Constraint - dyn_hair_l_01 (PP)"!(ComponentPose=@n10, Alpha=1, ConstraintWeights_0=0.5, ConstraintWeights_1:float=0.5) -> (Pose) {
    alphaBoolBlend {blendOption=Linear}
    boneToModify {boneName=dyn_hair_l_01}
    constraintSetup [{targetBone {boneName=head}}, {targetBone {boneName=neck_02}}]
    constraintWeights [1,1]
   }
   n12 = "Constraint - dyn_hair_l_02 (PP)"!(ComponentPose=@n11, Alpha=1, ConstraintWeights_0=1, ConstraintWeights_1:float=0) -> (Pose) {
    alphaBoolBlend {blendOption=Linear}
    boneToModify {boneName=dyn_hair_l_02}
    constraintSetup [{targetBone {boneName=head}}, {targetBone {boneName=neck_02}}]
    constraintWeights [1,1]
   }
   n13 = "Constraint - dyn_hair_l_03 (PPP)"!(ComponentPose=@n12, Alpha=1, ConstraintWeights_0=0.5, ConstraintWeights_1:float=0.5, ConstraintWeights_2:float=0) -> (Pose) {
    alphaBoolBlend {blendOption=Linear}
    boneToModify {boneName=dyn_hair_l_03}
    constraintSetup [{targetBone {boneName=neck_01}}, {targetBone {boneName=clavicle_l}}, {targetBone {boneName=head}}]
    constraintWeights [1,1,1]
   }
   n14 = "Constraint - dyn_hair_l_04 (P)"(ComponentPose=@n13, Alpha=1, ConstraintWeights_0=1) {
    alphaBoolBlend {blendOption=Linear}
    boneToModify {boneName=dyn_hair_l_04}
    constraintSetup [{targetBone {boneName=spine_05}}]
    constraintWeights [1]
   }
   n15 = "Constraint - dyn_hair_l_05 (PP)"!(ComponentPose=@n14, Alpha=1, ConstraintWeights_0=0.5, ConstraintWeights_1:float=0.5) -> (Pose) {
    alphaBoolBlend {blendOption=Linear}
    boneToModify {boneName=dyn_hair_l_05}
    constraintSetup [{targetBone {boneName=spine_04}}, {targetBone {boneName=spine_05}}]
    constraintWeights [1,1]
   }
   n16 = "Constraint - dyn_hair_c_01 (PP)"!(ComponentPose=@n15, Alpha=1, ConstraintWeights_0=0.5, ConstraintWeights_1:float=0.5) -> (Pose) {
    alphaBoolBlend {blendOption=Linear}
    boneToModify {boneName=dyn_hair_c_01}
    constraintSetup [{targetBone {boneName=head}}, {targetBone {boneName=neck_02}}]
    constraintWeights [1,1]
   }
   n17 = "Constraint - dyn_hair_c_02 (PP)"!(ComponentPose=@n16, Alpha=1, ConstraintWeights_0=1, ConstraintWeights_1:float=0) -> (Pose) {
    alphaBoolBlend {blendOption=Linear}
    boneToModify {boneName=dyn_hair_c_02}
    constraintSetup [{targetBone {boneName=head}}, {targetBone {boneName=neck_02}}]
    constraintWeights [1,1]
   }
   n18 = "Constraint - dyn_hair_c_03 (PP)"!(ComponentPose=@n17, Alpha=1, ConstraintWeights_0=0.5, ConstraintWeights_1:float=0.5) -> (Pose) {
    alphaBoolBlend {blendOption=Linear}
    boneToModify {boneName=dyn_hair_c_03}
    constraintSetup [{targetBone {boneName=neck_01}}, {targetBone {boneName=head}}]
    constraintWeights [1,1]
   }
   n19 = "Constraint - dyn_hair_c_04 (PP)"!(ComponentPose=@n18, Alpha=1, ConstraintWeights_0=0.5, ConstraintWeights_1:float=0.5) -> (Pose) {
    alphaBoolBlend {blendOption=Linear}
    boneToModify {boneName=dyn_hair_c_04}
    constraintSetup [{targetBone {boneName=spine_05}}, {targetBone {boneName=clavicle_l}}]
    constraintWeights [1,1]
   }
   n20 = "Constraint - dyn_hair_c_05 (PP)"!(ComponentPose=@n19, Alpha=1, ConstraintWeights_0=0.5, ConstraintWeights_1:float=0.5) -> (Pose) {
    alphaBoolBlend {blendOption=Linear}
    boneToModify {boneName=dyn_hair_c_05}
    constraintSetup [{targetBone {boneName=spine_04}}, {targetBone {boneName=spine_05}}]
    constraintWeights [1,1]
   }
   n21 = "Constraint - dyn_hair_side_r_01_s (PP)"!(ComponentPose=@n20, Alpha=1, ConstraintWeights_0=1, ConstraintWeights_1:float=0) -> (Pose) {
    alphaBoolBlend {blendOption=Linear}
    boneToModify {boneName=dyn_hair_side_r_01_s}
    constraintSetup [{targetBone {boneName=head}}, {targetBone {boneName=neck_02}}]
    constraintWeights [1,1]
   }
   n22 = "Constraint - dyn_hair_side_r_02_s (PP)"!(ComponentPose=@n21, Alpha=1, ConstraintWeights_0=0.5, ConstraintWeights_1:float=0.5) -> (Pose) {
    alphaBoolBlend {blendOption=Linear}
    boneToModify {boneName=dyn_hair_side_r_02_s}
    constraintSetup [{targetBone {boneName=head}}, {targetBone {boneName=neck_02}}]
    constraintWeights [1,1]
   }
   n23 = "Constraint - dyn_hair_side_r_03_s (PP)"!(ComponentPose=@n22, Alpha=1, ConstraintWeights_0=0.5, ConstraintWeights_1:float=0.5) -> (Pose) {
    alphaBoolBlend {blendOption=Linear}
    boneToModify {boneName=dyn_hair_side_r_03_s}
    constraintSetup [{targetBone {boneName=head}}, {targetBone {boneName=neck_02}}]
    constraintWeights [1,1]
   }
   n24 = "Constraint - dyn_hair_side_l_01 (PP)"!(ComponentPose=@n23, Alpha=1, ConstraintWeights_0=1, ConstraintWeights_1:float=0) -> (Pose) {
    alphaBoolBlend {blendOption=Linear}
    boneToModify {boneName=dyn_hair_side_l_01}
    constraintSetup [{targetBone {boneName=head}}, {targetBone {boneName=neck_02}}]
    constraintWeights [1,1]
   }
   n25 = "Constraint - dyn_hair_side_l_02 (PP)"!(ComponentPose=@n24, Alpha=1, ConstraintWeights_0=0.5, ConstraintWeights_1:float=0.5) -> (Pose) {
    alphaBoolBlend {blendOption=Linear}
    boneToModify {boneName=dyn_hair_side_l_02}
    constraintSetup [{targetBone {boneName=head}}, {targetBone {boneName=neck_02}}]
    constraintWeights [1,1]
   }
   n26 = "Constraint - dyn_hair_side_l_03 (PP)"!(ComponentPose=@n25, Alpha=1, ConstraintWeights_0=0.5, ConstraintWeights_1:float=0.5) -> (Pose) {
    alphaBoolBlend {blendOption=Linear}
    boneToModify {boneName=dyn_hair_side_l_03}
    constraintSetup [{targetBone {boneName=head}}, {targetBone {boneName=neck_02}}]
    constraintWeights [1,1]
   }
   n27 = "Constraint - dyn_hair_side_l_04 (PP)"!(ComponentPose=@n26, Alpha=1, ConstraintWeights_0=0.5, ConstraintWeights_1:float=0.5) -> (Pose) {
    alphaBoolBlend {blendOption=Linear}
    boneToModify {boneName=dyn_hair_side_l_04}
    constraintSetup [{targetBone {boneName=head}}, {targetBone {boneName=neck_02}}]
    constraintWeights [1,1]
   }
   n28 = "Constraint - dyn_hair_side_l_05 (PP)"!(ComponentPose=@n27, Alpha=1, ConstraintWeights_0=0.5, ConstraintWeights_1:float=0.5) -> (Pose) {
    alphaBoolBlend {blendOption=Linear}
    boneToModify {boneName=dyn_hair_side_l_05}
    constraintSetup [{targetBone {boneName=neck_01}}, {targetBone {boneName=clavicle_r}}]
    constraintWeights [1,1]
   }
   n29 = "Transform (Modify) Bone - Bone: dyn_hair_c_03"(ComponentPose=@n28, Alpha=1, Translation="0,0,2", Rotation="0,0,0", Scale="1,1,1") {translationMode=BMM_Additive, translationSpace=BCS_BoneSpace, boneToModify {boneName=dyn_hair_c_03}, translation {z=2}}
   n30 = "Transform (Modify) Bone - Bone: dyn_hair_l_03"(ComponentPose=@n29, Alpha=1, Translation="0,0,2", Rotation="0,0,0", Scale="1,1,1") {translationMode=BMM_Additive, translationSpace=BCS_BoneSpace, boneToModify {boneName=dyn_hair_l_03}, translation {z=2}}
   n31 = "Transform (Modify) Bone - Bone: dyn_hair_r_03"(ComponentPose=@n30, Alpha=1, Translation="0,0,2", Rotation="0,0,0", Scale="1,1,1") {translationMode=BMM_Additive, translationSpace=BCS_BoneSpace, boneToModify {boneName=dyn_hair_r_03}, translation {z=2}}
   n32 = NOT Boolean(A=Get bRigidBodyEnabled())
   n33 = Get Is Wearing Cape() -> (Is Wearing Cape:bool)
   n34 = Get isWearingBackbling()
   Output Pose(Result=Blend Poses by bool(
     BlendPose_0=Component To Local(ComponentPose=@n31)
     BlendPose_1=Use cached pose 'Start_Cache'()
     bActiveValue=OR Boolean!(A=@n34, B=@n33, C:bool=@n32) -> (ReturnValue)
    ) {blendType=Linear}) {layerGroup=None}
   n35 = Layered blend per bone(
    BasePose=Copy Pose From Mesh() {bUseAttachedParent=true}
    BlendPoses_0=Copy Pose From Mesh(SourceMeshComponent="Get Face Mesh "() -> ("Face Mesh ":Skeletal Mesh Component Object Reference)) {bUseAttachedParent=true}
   ) {layerSetup [{branchFilters [
       {boneName=FACIAL_C_FacialRoot}
       {boneName=FACIAL_L_NasolabialBulge}
       {boneName=FACIAL_R_NasolabialBulge}
       {boneName=FACIAL_L_NasolabialFurrow}
       {boneName=FACIAL_R_NasolabialFurrow}
       {boneName=FACIAL_L_CheekLower}
       {boneName=FACIAL_R_CheekLower}
       {boneName=FACIAL_C_Nose}
       {boneName=FACIAL_C_NoseLower}
       {boneName=FACIAL_L_Nostril}
       {boneName=FACIAL_R_Nostril}
       {boneName=FACIAL_C_NoseTip}
       {boneName=FACIAL_C_MouthUpper}
       {boneName=FACIAL_C_LipUpper}
       {boneName=FACIAL_L_LipUpper}
       {boneName=FACIAL_R_LipUpper}
       {boneName=FACIAL_L_LipUpperOuter}
       {boneName=FACIAL_R_LipUpperOuter}
       {boneName=FACIAL_L_LipCorner}
       {boneName=FACIAL_R_LipCorner}
       {boneName=FACIAL_L_Masseter}
       {boneName=FACIAL_R_Masseter}
       {boneName=FACIAL_C_UnderChin}
       {boneName=FACIAL_L_UnderChin}
       {boneName=FACIAL_R_UnderChin}
       {boneName=FACIAL_C_LowerLipRotation}
       {boneName=FACIAL_C_MouthLower}
       {boneName=FACIAL_C_LipLower}
       {boneName=FACIAL_L_LipLower}
       {boneName=FACIAL_R_LipLower}
       {boneName=FACIAL_L_LipLowerOuter}
       {boneName=FACIAL_R_LipLowerOuter}
       {boneName=FACIAL_C_Jaw}
       {boneName=FACIAL_C_Jawline}
       {boneName=FACIAL_L_ChinSide}
       {boneName=FACIAL_R_ChinSide}
       {boneName=FACIAL_C_Chin}
       {boneName=FACIAL_L_Ear}
       {boneName=FACIAL_R_Ear}
       {boneName=FACIAL_L_Jawline}
       {boneName=FACIAL_R_Jawline}
      ]}]}
   n36 = Control Rig!(Source=@n35, Alpha, isTargeting:bool=Get bIsTargeting()) -> (Pose) {controlRigAssetReference {controlRigAsset=&/M_MED_NoiseClueLurch_FaceAcc_CtrlRig}}
   n37 = RigidBody(ComponentPose=Local To Component(LocalPose=@n36), OverrideWorldGravity=@n1) {
    bClampLinearTranslationLimitToRefPose=true
    bOverrideWorldGravity=true
    overridePhysicsAsset=&M_MED_NoiseClueLurch_FaceAcc_Physics
    simulationSpace=BaseBoneSpace
    alphaBoolBlend {blendOption=Linear}
    baseBoneRef {boneName=neck_01}
    componentAppliedLinearAccClamp=<0 0 0>
    overrideWorldGravity {z=-980}
    simSpaceSettings {
     dampingAlpha=0
     maxAngularAcceleration=16
     maxAngularVelocity=4
     maxLinearAcceleration=1000
     maxLinearVelocity=250
     velocityScaleZ=0.5
     worldAlpha=1
    }
   }
   Save cached pose 'Start_Cache'(Pose=Component To Local(ComponentPose=@n37)) {cacheName=Start_Cache}
  ]}
 EventGraph {nodes [
   Event BlueprintUpdateAnimation()
   > n1 = Sequence()
   > n2 = "Gravity Override State Function (Parameter Asset) (Nativized)"(
    CharacterPartAnimInstance=Self-Reference()
    GameTimeInSeconds=GetGameTimeInSeconds()
    Parameters=&/M_MED_NoiseClueLurch_FaceAcc_Parameters
   )
   > Set GravityOverride(GravityOverride=@n2)
   > Set RBAN Settings Out(RBAN Settings Out=@n2.OutRigidBodyAnimNodeParameters)
   n3 = CheckForBackbling()
   Set isWearingBackbling(
    execute=@n1.then_1
    isWearingBackbling=AND Boolean!(A=@n3, B=NOT Boolean(A=@n3.bIsBackblingACape), C:bool=NOT Boolean(A=@n3.bIsBackblingLong)) -> (ReturnValue)
   )
   > Set Is Wearing Cape(Is Wearing Cape=OR Boolean!(A=@n3.bIsBackblingACape, B=@n3.bIsBackblingLong) -> (ReturnValue))
   Event OnCharacterCustomizationLoaded()
   > n4 = Cast To PlayerPawn_Athena(Object=TryGetPawnOwner())
   > "Set Face Mesh "("Face Mesh ":Skeletal Mesh Component Object Reference=GetSkeletalMeshForPartType(self=@n4, PartType=Head)) -> (Output_Get:Skeletal Mesh Component Object Reference)
   Sequence()
   > n5 = Branch(Condition=CharacterItemDefinitionHasAnyMetaTag(Tags=Make Array!("[0]"=(TagName="Cosmetics.Variant.Property.Stage1"), "[1]":Gameplay Tag Structure=(TagName="Cosmetics.Variant.Property.Stage2")) -> (Array)))
   > Set isSunGlassesON(isSunGlassesON:bool=true) -> (Output_Get:bool)
   Set isSunGlassesON(execute=@n5.else, isSunGlassesON=false)
  ]}
]
variables [
 "Face Mesh " {type=SkeletalMeshComponent}
 GravityOverride {type=Vector}
 Is Wearing Cape {type=boolean}
 isSunGlassesON {type=boolean}
 isWearingBackbling {type=boolean}
 RBAN Settings Out {type=FortRigidBodyAnimNodeParameters}
]
